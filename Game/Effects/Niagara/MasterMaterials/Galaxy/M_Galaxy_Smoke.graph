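# Translucent unlit Niagara sprite master material for galaxy-themed smoke, showing a nebula and starfield inside the particles.
# Camera-vector triplanar projection maps a time-rotated, desaturation-adjustable nebula plus rotating small stars and bright stars, so the galaxy is independent of sprite UVs.
# A panning star-texture mask makes bright stars twinkle, then particle color tints the result before emissive adjustments.
# Particle alpha times a cloud mask drives depth-faded opacity.
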Material /Game/Effects/Niagara/MasterMaterials/Galaxy/M_Galaxy_Smoke
graph {
 properties {bUsedWithNiagaraSprites=true, blendMode=BLEND_Translucent, pixelDepthOffsetMode=PDOM_Legacy, shadingModel=MSM_Unlit, usedShadingModels=Unlit}
 nodes [
  n1 = ParticleColor()
  DepthFade(
   Opacity=Multiply(
    A=@n1.A
    B=TextureSample() {samplerType=SAMPLERTYPE_LinearGrayscale, texture=/Game/Athena/Apollo/Environments/Blueprints/T_Cloud_MacroMask}.G
   )
   FadeDistance=ScalarParameter() {defaultValue=40, parameterName=FadeDistance}
  )
  n2 = ScalarParameter() {defaultValue=1, group=Galaxy, parameterName=GalaxyTiling}
  n3 = Constant()
  n4 = ScalarParameter() {defaultValue=6, group=Galaxy, parameterName=GalaxyAxisFade}
  n5 = Time()
  n6 = Constant()
  n7 = ScalarParameter() {defaultValue=0.02, group=Galaxy, parameterName=Galaxy Rotation Speed}
  n8 = ScalarParameter() {defaultValue=-0.02, group=Galaxy, parameterName=Galaxy Rotation Speed_1}
  n9 = Multiply(
   A=LinearInterpolate(A=@n5, B=Multiply(A=@n5, B=@n6), Alpha=ScalarParameter() {group=Galaxy, parameterName=Background Rotation})
   B=AppendVector(A=AppendVector(A=@n7, B=@n8), B=ScalarParameter() {defaultValue=0.1, group=Galaxy, parameterName=Galaxy Rotation Speed_2})
  )
  n10 = TextureObjectParameter() {
   group=Galaxy
   parameterName=GalaxyTexture
   texture=/Game/Characters/Player/Male/Medium/Bodies/M_Med_Soldier_04/Skins/Galaxy/Textures/T_FN_Nebula
  }
  n11 = Power(Base=Desaturation(
    None=MaterialFunctionCall!(CameraVector, Rotation Speed=@n9, Texture=@n10, Tiling=@n2, Offset=@n3, AxisFadeContrast=@n4) {materialFunction=/Game/Weapons/WeaponSkins/MaterialFunction/TriplanarCameraVector_Rotated}.Result
    Fraction=ScalarParameter() {group=Galaxy, parameterName=Galaxy Saturation}
   )) {constExponent=0.7}
  n12 = Desaturation(None=@n11, Fraction=ScalarParameter() {defaultValue=0.5, group=Galaxy, parameterName=Star Saturation})
  n13 = Time()
  n14 = ScalarParameter() {defaultValue=-0.125, group=Galaxy, parameterName=Small Star Rotation speed}
  n15 = TextureObject() {samplerType=SAMPLERTYPE_Masks, texture=/Game/Characters/Player/Male/Medium/Bodies/M_Med_Soldier_04/Skins/Galaxy/Textures/Celestial_T_Stars}
  n16 = ComponentMask(None=MaterialFunctionCall(
    Rotation Speed=Multiply(A=@n13, B=@n14)
    Texture=@n15
    Tiling=ScalarParameter() {defaultValue=6, group=Galaxy, parameterName=Small Star Tiling}
   ) {materialFunction=&TriplanarCameraVector_Rotated}.Result) {b=true}
  n17 = LinearInterpolate(
   A=@n11
   B=Multiply(A=@n12, B=VectorParameter() {group=Galaxy, parameterName=Star Brightness, defaultValue=<rgba 300 300 300 1>}.RGB)
   Alpha=Multiply(A=@n16, B=@n16)
  )
  n18 = Multiply(A=@n12, B=VectorParameter() {group=Galaxy, parameterName=Star Brightness_1, defaultValue=<rgba 950 950 950 1>}.RGB)
  n19 = TextureObject() {samplerType=SAMPLERTYPE_Masks, texture=&Celestial_T_Stars}
  n20 = ComponentMask(None=MaterialFunctionCall!(
    CameraVector
    Texture=@n19
    Tiling=ScalarParameter() {defaultValue=6, group=Galaxy, parameterName=Bright Star Tiling}
    Offset
    AxisFadeContrast
   ) {materialFunction=/Engine/Functions/Engine_MaterialFunctions01/Texturing/TriplanarCameraVector}.Result) {r=true}
  n21 = TextureSample(UVs=Panner(Coordinate=TextureCoordinate() {uTiling=2, vTiling=2}) {speedX=0.133, speedY=0.2}) {
   samplerSource=SSM_Wrap_WorldGroupSettings
   samplerType=SAMPLERTYPE_Masks
   texture=&Celestial_T_Stars
  }
  n22 = Multiply(A=Multiply(A=@n20, B=@n20), B=@n21.G)
  MaterialFunctionCall(Emissive=Multiply(A=LinearInterpolate(A=@n17, B=@n18, Alpha=@n22), B=@n1.RGB)) {materialFunction=&MF_FX_EmissiveAdjustments}
 ]
}
